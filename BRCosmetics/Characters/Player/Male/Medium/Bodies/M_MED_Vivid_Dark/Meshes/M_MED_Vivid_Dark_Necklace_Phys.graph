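# Physics asset for the necklace on the M_MED_Vivid_Dark male player cosmetic body, apparently for secondary motion against the upper torso.
# Capsules on spine_02, spine_04, spine_05 and both clavicles approximate the chest and shoulders.
# A small master capsule links spine_04 to a two-box chain, dyn_necklace_01 and dyn_necklace_end.
# The two chain joints have tight swing and twist limits, while all other constraints have free angular motion.
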PhysicsAsset /BRCosmetics/Characters/Player/Male/Medium/Bodies/M_MED_Vivid_Dark/Meshes/M_MED_Vivid_Dark_Necklace_Phys
bodies {
 clavicle_l [{shapeType=Capsule, length=6.758, radius=8.5617, shapeName=clavicle_r_capsule, center=<9.5243 2.9496 -2.3683>, rotation=<rot 90 0 5.0002>}]
 clavicle_r [{shapeType=Capsule, length=6.758, radius=8.5617, shapeName=clavicle_r_capsule, center=<-9.5243 -2.9496 2.3683>, rotation=<rot 90 0 5.0002>}]
 dyn_necklace_01 [{
   shapeType=Box
   extentX=1.5881
   extentY=8.8113
   extentZ=13.432
   shapeName=dyn_necklace_01_box
   center=<10.5595 0.0177 -0.0263>
   rotation=<rot -0.059 -90 0.3202>
  }]
 dyn_necklace_MSTR [{shapeType=Capsule, length=0.505, radius=0.505, shapeName=dyn_necklace_MSTR_capsule, rotation=<rot 0 -90 0>}]
 dyn_necklace_end [{
   shapeType=Box
   extentX=1.1823
   extentY=9.391
   extentZ=11.5967
   shapeName=dyn_necklace_end_box
   center=<5.0121 -0.129 -0.0233>
   rotation=<rot 0 89.3741 -89.3863>
  }]
 spine_02 [{shapeType=Capsule, length=6.9694, radius=12.4965}]
 spine_04 [
  {shapeType=Capsule, length=5.3225, radius=13.1297, shapeName=spine_04_capsule, center=<2.6144 1.1423 0>, rotation=<rot 0 -10.0007 0>}
  {shapeType=Capsule, length=16, radius=14, shapeName=spine_04_capsule, center=<-2.4207 0 0>}
 ]
 spine_05 [{shapeType=Capsule, length=10.8212, radius=11.1242, shapeName=spine_05_capsule, center=<-2.2348 1.3357 0>}]
}
constraints [
 {bone1Name=dyn_necklace_01, bone2Name=dyn_necklace_MSTR, swing1LimitDegrees=35, swing2LimitDegrees=20, twistLimitDegrees=0}
 {
  bone1Name=clavicle_r
  bone2Name=spine_05
  swing1LimitDegrees=45
  swing1Motion=Free
  swing2LimitDegrees=45
  swing2Motion=Free
  twistLimitDegrees=45
  twistMotion=Free
 }
 {
  bone1Name=spine_05
  bone2Name=spine_04
  swing1LimitDegrees=45
  swing1Motion=Free
  swing2LimitDegrees=45
  swing2Motion=Free
  twistLimitDegrees=45
  twistMotion=Free
 }
 {
  bone1Name=dyn_necklace_MSTR
  bone2Name=spine_04
  swing1LimitDegrees=45
  swing1Motion=Free
  swing2LimitDegrees=45
  swing2Motion=Free
  twistLimitDegrees=45
  twistMotion=Free
 }
 {
  bone1Name=spine_04
  bone2Name=spine_02
  swing1LimitDegrees=45
  swing1Motion=Free
  swing2LimitDegrees=45
  swing2Motion=Free
  twistLimitDegrees=45
  twistMotion=Free
 }
 {
  bone1Name=clavicle_l
  bone2Name=spine_05
  swing1LimitDegrees=45
  swing1Motion=Free
  swing2LimitDegrees=45
  swing2Motion=Free
  twistLimitDegrees=45
  twistMotion=Free
 }
 {bone1Name=dyn_necklace_end, bone2Name=dyn_necklace_01, swing1LimitDegrees=25, swing2LimitDegrees=35, twistLimitDegrees=5}
]
