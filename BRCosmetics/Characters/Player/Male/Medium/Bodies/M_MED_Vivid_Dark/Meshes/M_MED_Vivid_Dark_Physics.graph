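# Physics asset for the M_MED_Vivid_Dark male medium player body, set up around a dangling belt piece.
# A chain of constraints links dyn_belt_MSTR through four belt bones, with capsules on the upper links and boxes on the last two.
# Swing and twist limits generally widen toward the free end, so the tip moves more freely.
# Unconstrained capsules on spine_02 and both thigh twist bones supply body volumes around the belt.
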PhysicsAsset /BRCosmetics/Characters/Player/Male/Medium/Bodies/M_MED_Vivid_Dark/Meshes/M_MED_Vivid_Dark_Physics
bodies {
 dyn_belt_01 [{shapeType=Capsule, length=3.12, radius=1.375, shapeName=dyn_belt_01_capsule, center=<2.8009 0.5988 0>, rotation=<rot 90 0 0>}]
 dyn_belt_02 [{shapeType=Capsule, length=2.8756, radius=1.5, shapeName=dyn_belt_02_capsule, center=<2.4121 0.2164 0>, rotation=<rot 90 0 0>}]
 dyn_belt_03 [{shapeType=Box, extentX=2.51, extentY=4.76, extentZ=1.01, shapeName=dyn_belt_03_box, center=<2.5759 0 0>, rotation=<rot 0 -90 0>}]
 dyn_belt_04 [{
   shapeType=Box
   extentX=3.01
   extentY=1.01
   extentZ=4.26
   shapeName=dyn_belt_04_box
   center=<1.9663 0.3623 -0.1937>
   rotation=<rot -4.9924 -87.119 -86.8419>
  }]
 dyn_belt_MSTR [{shapeType=Capsule, length=0.505, radius=0.505, shapeName=dyn_belt_MSTR_capsule, rotation=<rot 0 -90 0>}]
 spine_02 [{shapeType=Capsule, length=6.9694, radius=12.4965}]
 thigh_twist_01_l [{shapeType=Capsule, length=30.125, radius=10.25, rotation=<rot 90 0 0>}]
 thigh_twist_01_r [{shapeType=Capsule, length=30.875, radius=9.5, rotation=<rot 90 0 0>}]
}
constraints [
 {bone1Name=dyn_belt_01, bone2Name=dyn_belt_MSTR, swing1LimitDegrees=20, swing2LimitDegrees=20, twistLimitDegrees=5}
 {bone1Name=dyn_belt_02, bone2Name=dyn_belt_01, swing1LimitDegrees=30, swing2LimitDegrees=30, twistLimitDegrees=5}
 {bone1Name=dyn_belt_03, bone2Name=dyn_belt_02, swing1LimitDegrees=45, swing2LimitDegrees=45, twistLimitDegrees=8}
 {bone1Name=dyn_belt_04, bone2Name=dyn_belt_03, swing1LimitDegrees=45, swing2LimitDegrees=45, twistLimitDegrees=10}
]
